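# Niagara cosmetic effect for the dark Wizard-Witch character's hands, drawing dark blue-to-purple flipbook sprites at skeletal sockets.
# Two emitters partition particles across sockets, following each with a per-socket offset.
# Backward plays short-lived, shrinking sprites with a reversed flipbook; Forward plays longer-lived, growing sprites from random start frames.
# Both fade over life, animate camera offset, and use User.bisHandHidden to zero spawn rate and sprite size.
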NiagaraSystem /BRCosmetics/Effects/Fort_Effects/Effects/Characters/Athena_Parts/Wizard-Witch_Dark/NS_Witch-WizardDark_Hands
dependencies {
 usedDataInterfaces [&NiagaraDataInterfaceArrayFloat3, &NiagaraDataInterfaceCurve, &NiagaraDataInterfaceSocketReader]
 usedDynamicInputs [&SimulationPosition]
 usedModules [
  &self:SystemSpawnScript.SystemScriptSource.SystemScriptGraph.NiagaraNodeAssignment_0.SetVariables_#1
  &SystemState
 ]
}
emitters {
 Backward {
  emitterUpdateScript [
   &EmitterState(Life Cycle Mode=::NewEnumerator1, UseLoopDelay=false)
   &SpawnRate(
    Spawn Probability=0.6
    SpawnRate=&MakeCustomFloatFromBool(Bool=$User.bisHandHidden, True Float=0, False Float=12)
    Use Spawn Probability=true
   )
  ]
  particleSpawnScript [
   &PartitionParticles()
   &InitializeParticle(
    Color=&RandomRangeLinearColor(Minimum=<rgba 0.0006 0.0075 0.0225 4>, Maximum=<rgba 0.013 0 0.03 8>)
    Lifetime=&RandomRangeFloat(Minimum=1.5, Maximum=2.35)
    Sprite Rotation Mode=::NewEnumerator2
    Sprite Size Mode=::NewEnumerator1
    Uniform Sprite Size Max=11
    Uniform Sprite Size Min=9
    UsePositionOffset=false
    Write Material Random=true
   ) {module=InitializeParticle001}
   &DynamicMaterialParameters(
    Index 0 Param 1=&RandomRangeFloat(Minimum=0, Maximum=1)
    Index 0 Param 2=&RandomRangeFloat(Minimum=0.2, Maximum=1)
    Index 0 Param 3=6
    Module.Index0Float0Write=true
    Module.Index0Float1Write=true
    Module.Index0Float2Write=true
    Module.Index0Float3Write=true
   ) {module=DynamicMaterialParameters001}
  ]
  particleUpdateScript [
   &ParticleState()
   &SocketLocation(
    SamplePosOffset=true
    Sampled Position Offset=&SelectVectorFromArray(
     Direct Array Index=$Particles.PartitionParticles.Partition
     Vector Selection Array=DataInterface() {propertyValues {FloatData [<8 -4 0>, <-6 2 0>]}}
    )
    Socket Index=$Particles.PartitionParticles.Partition
    Socket Reader=$System.Socket Reader
    Socket Sampling Mode=::NewEnumerator3
    VelocitySpeedLimit=false
   )
   &SubUVAnimation(
    End Frame:NiagaraInt32=&RandomRangeInt(Minimum=16, Maximum=16)
    Start Frame=63
    SubUV Animation Mode:ENiagara_SubUVLookupMode=::NewEnumerator1
    SubUV Curve=DataInterface() {propertyValues {Curve {keys [<at 0 0 RCIM_Cubic>, {arriveTangent=1.7187, interpMode=RCIM_Cubic, leaveTangent=1.7187, tangentMode=RCTM_User, time=1, value=1}]}}}
   )
   &ScaleColor(
    Scale Alpha=&FloatFromCurve(FloatCurve=DataInterface() {propertyValues {Curve {keys [
         {arriveTangent=-0.7111, interpMode=RCIM_Cubic, leaveTangent=1.7199, tangentMode=RCTM_User, time=0, value=0}
         {arriveTangent=-0.2125, interpMode=RCIM_Cubic, leaveTangent=-5.4535, tangentMode=RCTM_Break, time=0.41, value=1}
         {interpMode=RCIM_Cubic, leaveTangent=-0.2648, time=0.85, value=0}
        ]}}})
    ScaleA=true
    ScaleRGB=true
   )
   &ScaleSpriteSize(Scale Factor=&Vector2DFromFloat(Value=&FloatFromCurve(FloatCurve=DataInterface() {propertyValues {Curve {keys [<at 0 2 RCIM_Cubic>, {arriveTangent=-1.9293, interpMode=RCIM_Cubic, leaveTangent=4.1558, tangentMode=RCTM_User, time=1, value=0.3}]}}})))
   &CameraOffset(Camera Offset Amount=&FloatFromCurve(FloatCurve=DataInterface() {propertyValues {Curve {keys [<at 0 12>,<at 1 0>]}}}))
   &ScaleSpriteSize(
    Scale Sprite Size Mode=::NewEnumerator0
    Uniform Scale Factor=&MakeCustomFloatFromBool(Bool=$User.bisHandHidden, True Float=0, False Float=1)
   ) {module=ScaleSpriteSize001}
  ]
  rendererClasses [&NiagaraSpriteRendererProperties]
  renderers [{rendererClass=&NiagaraSpriteRendererProperties}]
 }
 Forward {
  emitterUpdateScript [
   &EmitterState(Life Cycle Mode=::NewEnumerator1, UseLoopDelay=false)
   &SpawnRate(
    Spawn Probability=0.6
    SpawnRate=&MakeCustomFloatFromBool(Bool=$User.bisHandHidden, True Float=0, False Float=6)
    Use Spawn Probability=true
   )
  ]
  particleSpawnScript [
   &PartitionParticles()
   &InitializeParticle(
    Color=&RandomRangeLinearColor(Minimum=<rgba 0.0003 0.0033 0.01 1.5>, Maximum=<rgba 0.0121 0 0.028 2>)
    Lifetime=&RandomRangeFloat(Minimum=5, Maximum=6)
    Sprite Rotation Mode=::NewEnumerator2
    Sprite Size Mode=::NewEnumerator1
    Uniform Sprite Size Max=16
    Uniform Sprite Size Min=12
    UsePositionOffset=false
    Write Material Random=true
   ) {module=InitializeParticle001}
   &DynamicMaterialParameters(
    Index 0 Param 1=&RandomRangeFloat(Minimum=0, Maximum=1)
    Index 0 Param 2=&RandomRangeFloat(Minimum=0, Maximum=1)
    Index 0 Param 3=6
    Module.Index0Float0Write=true
    Module.Index0Float1Write=true
    Module.Index0Float2Write=true
    Module.Index0Float3Write=true
   ) {module=DynamicMaterialParameters001}
   &CameraOffset(Camera Offset Amount=&RandomRangeFloat(Minimum=-6, Maximum=0))
  ]
  particleUpdateScript [
   &SocketLocation(
    SamplePosOffset=true
    Sampled Position Offset=&SelectVectorFromArray(
     Direct Array Index=$Particles.PartitionParticles.Partition
     Vector Selection Array=DataInterface() {propertyValues {FloatData [<8 -4 0>, <-6 2 0>]}}
    )
    Socket Index=$Particles.PartitionParticles.Partition
    Socket Reader=$System.Socket Reader
    Socket Sampling Mode=::NewEnumerator3
    VelocitySpeedLimit=false
   )
   &ParticleState()
   &SubUVAnimation(
    Start Frame:NiagaraInt32=&RandomRangeInt(Minimum=0, Maximum=36)
    SubUV Animation Mode:ENiagara_SubUVLookupMode=::NewEnumerator1
    SubUV Curve=DataInterface() {propertyValues {Curve {keys [{arriveTangent=1.5316, interpMode=RCIM_Cubic, leaveTangent=1.5316, tangentMode=RCTM_User, time=0, value=0}, <at 1 1>]}}}
   )
   &ScaleColor(
    Scale Alpha=&FloatFromCurve(FloatCurve=DataInterface() {propertyValues {Curve {keys [
         {arriveTangent=-0.9958, interpMode=RCIM_Cubic, leaveTangent=2.9799, tangentMode=RCTM_User, time=0, value=0}
         {arriveTangent=0.0103, interpMode=RCIM_Cubic, leaveTangent=-2.9149, tangentMode=RCTM_Break, time=0.5, value=1}
         {interpMode=RCIM_Cubic, leaveTangent=-0.2648, time=1, value=0}
        ]}}})
    ScaleA=true
    ScaleRGB=true
   )
   &ScaleSpriteSize(Scale Factor=&Vector2DFromFloat(Value=&FloatFromCurve(FloatCurve=DataInterface() {propertyValues {Curve {keys [
          {arriveTangent=2.3165, interpMode=RCIM_Cubic, leaveTangent=0.5677, tangentMode=RCTM_User, time=0, value=1}
          {arriveTangent=2.3533, interpMode=RCIM_Cubic, leaveTangent=7.6663, tangentMode=RCTM_User, time=1, value=3}
         ]}}})))
   &CameraOffset(Camera Offset Amount=&Lerp_Float(
     A=$Particles.CameraOffset
     B=&Lerp_Float(A=-3, B=12, Alpha=$Particles.Initial.MaterialRandom)
     Alpha=$Particles.NormalizedAge
    )) {module=CameraOffset001}
   &ScaleSpriteSize(
    Scale Sprite Size Mode=::NewEnumerator0
    Uniform Scale Factor=&MakeCustomFloatFromBool(Bool=$User.bisHandHidden, True Float=0, False Float=1)
   ) {module=ScaleSpriteSize001}
  ]
  rendererClasses [&NiagaraSpriteRendererProperties]
  renderers [{rendererClass=&NiagaraSpriteRendererProperties}]
 }
}
summary {
 systemName=NS_Witch-WizardDark_Hands
 emitters [
  {emitterName=Backward, rendererClasses [&NiagaraSpriteRendererProperties]}
  {emitterName=Forward, rendererClasses [&NiagaraSpriteRendererProperties]}
 ]
 userVariables [User.bisHandHidden {type=NiagaraBool}]
}
