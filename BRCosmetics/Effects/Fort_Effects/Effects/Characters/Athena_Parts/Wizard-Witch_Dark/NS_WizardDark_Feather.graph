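# Niagara effect for the dark wizard/witch cosmetic's feather that crackles magenta lightning between points on a skeletal mesh.
# BigLightningEndPicker spawns short-lived flickering sprite and light particles in randomly delayed bursts, sampling two random bone/socket positions for each.
# BigLightningRibbon reads those particles through an attribute reader, scaling its spawn rate by their count.
# Ribbon points interpolate between the sampled endpoints with mid-arc jitter, then drift, thin and dim.
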NiagaraSystem /BRCosmetics/Effects/Fort_Effects/Effects/Characters/Athena_Parts/Wizard-Witch_Dark/NS_WizardDark_Feather
dependencies {
 usedDataInterfaces [&NiagaraDataInterfaceParticleRead, &NiagaraDataInterfaceCurve, &NiagaraDataInterfaceColorCurve]
 usedDynamicInputs [&MakeVector4_FromFloat]
 usedModules [&SystemState]
}
emitters {
 BigLightningEndPicker {
  emitterUpdateScript [
   /SetVariables(Emitter.SetLoopDelay:NiagaraFloat=&UniformRangedFloat(Minimum=0.4666, Maximum=1.1)) {isSetParameters=true}
   /SetVariables(Emitter.SetLoopDuration:NiagaraFloat=&Subtract_Float(A=1, B=$Emitter.SetLoopDelay)) {isSetParameters=true}
   &EmitterState(
    Life Cycle Mode=::NewEnumerator1
    Loop Delay=$Emitter.SetLoopDelay
    Loop Duration=$Emitter.SetLoopDuration
    Recalculate Duration Each Loop=true
    UseLoopDelay=true
   )
   &SpawnRate(SpawnRate=5, Use Spawn Probability=false)
  ]
  particleSpawnScript [/Niagara/Modules/Spawn/Initialization/InitializeParticle(
    Color=&MultiplyLinearColor(
     Linear Color=<rgba 6 0 3.9 1>
     Scale Factor=&MakeVector4(X=1, Y=1, Z=1, W=&RandomRangeFloat(Minimum=0.75, Maximum=1.5))
    )
    Lifetime=&UniformRangedFloat(Minimum=0.2, Maximum=0.6)
    Module.Write Color=true
    Module.Write Lifetime=true
    Module.Write Mass=true
    Module.Write Material Random=true
    Module.Write Position=false
    Module.Write Scale=false
    Module.Write SpriteRotation=true
    Module.Write SpriteSize=true
    Position=&SimulationPosition()
    Sprite Rotation=90
    Sprite Size=&Vector2DFromFloat(Value=5)
   )]
  particleUpdateScript [
   &SkeletalMeshLocation(
    Bone / Socket Index=&RandomRangeInt(Minimum=0, Maximum=2)
    Bone Sampling Mode=::NewEnumerator4
    Inherit Velocity=false
    MaskSpawnGroup=false
    Skeletal Mesh=$User.SkeletalMesh
   ) {module=SkeletalMeshLocation001}
   /SetVariables(Particles.HomePosition:Vector3f=$Output.SkeletalMeshLocation001.SampledPosition) {isSetParameters=true}
   &SkeletalMeshLocation(
    Bone / Socket Index=&RandomRangeInt(Minimum=0, Maximum=2)
    Bone Sampling Mode=::NewEnumerator4
    Inherit Velocity=false
    MaskSpawnGroup=false
    Skeletal Mesh=$User.SkeletalMesh
   )
   /SetVariables(Particles.FriendPosition:Vector3f=$Output.SkeletalMeshLocation.SampledPosition) {isSetParameters=true}
   &SolveForcesAndVelocity(Clamp Velocity=false, Limit Acceleration=false)
   &ParticleState()
   &ScaleColor(
    Scale RGB=&VectorFromFloat(Value=&FloatFromCurve(
      FloatCurve=DataInterface() {propertyValues {Curve {keys [<at 0 1>,<at 0.0779 0.5708>,<at 0.1117 -0.2574>,<at 0.1722 0.7862>,<at 0.295 -0.2574>,<at 0.3419 0.5643>,<at 0.5191 0.1498>,<at 0.5473 -0.754>,<at 0.6267 0.8648>,<at 0.7996 0.2507>,<at 0.9136 0.9747>,<at 1 0>]}}}
      CurveIndex=&Multiply_Float(A=$Particles.NormalizedAge, B=&Multiply_Float(A=1, B=1))
     ))
    ScaleA=false
    ScaleRGB=true
   )
   /SetVariables(Particles.LightPosition:Vector3f=&Add_Vector(A=Linked() {linkedVariable {name=Particles.Position, type=NiagaraPosition}}, B=<0 -7 0>)) {isSetParameters=true}
  ]
  rendererClasses [&NiagaraSpriteRendererProperties, &NiagaraLightRendererProperties]
  renderers [
   {rendererClass=&NiagaraSpriteRendererProperties}
   {rendererClass=&NiagaraSpriteRendererProperties, rendererIndex=1}
   {rendererClass=&NiagaraLightRendererProperties, rendererIndex=2}
  ]
 }
 BigLightningRibbon {
  emitterSpawnScript [/Game/Effects/Niagara/Modules/Attributes/FN_Source_v2(AttributeReader=DataInterface() {propertyValues {EmitterBinding {bindingMode=Other, emitterName=BigLightningEndPicker}}})]
  emitterUpdateScript [
   &EmitterState()
   /SetVariables(Emitter.ParticlesPerRibbon:NiagaraFloat=&UniformRangedFloat(Minimum=20, Maximum=20)) {isSetParameters=true}
   &SpawnRate(
    SpawnRate=/Game/Effects/Niagara/DynamicInputs/Attributes/FN_GetEmitterParticleCount(AttributeReader:NiagaraDataInterfaceParticleRead=$Emitter.AttributeReader, Multiplier:NiagaraFloat=$Emitter.ParticlesPerRibbon)
    Use Spawn Probability=false
   )
  ]
  particleSpawnScript [
   &self:GetInfoFromParent() {module=GetInfoFromParent}
   /SetVariables(Particles.AgeByParent:NiagaraFloat=&FloatFromCurve(
     FloatCurve=DataInterface() {propertyValues {Curve {keys [
         {arriveTangent=-17.8847, interpMode=RCIM_Cubic, leaveTangent=-5.4543, tangentMode=RCTM_User, time=0.2, value=1}
         {arriveTangent=-4.0175, leaveTangent=-4.0175, time=0.8, value=0}
        ]}}}
     CurveIndex=$Particles.AgeImported
    )) {isSetParameters=true}
   /SetVariables(Particles.DistanceBetweenParents:NiagaraFloat=&ClampFloat(Float=&Add_Float(
      A=-0.5
      B=&Multiply_Float(
       A=1
       B=/Niagara/DynamicInputs/TypeConversions/VectorLengthSafe(Vector:Vector3f=&Subtract_Vector(A=$Particles.ParentPosition, B=$Particles.ParentFriendPosition))
      )
     ))) {isSetParameters=true}
   &InitializeRibbon(
    Color=<rgba 2.4952 0 2 1>
    Lifetime=&RandomRangeFloat(Minimum=0.333, Maximum=0.333)
    Module.Write Color=true
    Module.Write Lifetime=true
    Module.Write Mass=false
    Module.Write Material Random=true
    Module.Write Position=false
    Module.Write RibbonTwist=false
    Module.Write RibbonWidth=true
    Ribbon Width=&Multiply_Float(
     A=/Niagara/DynamicInputs/Math/LerpMultipleFloats(
      Float Count:ENiagara_MultipleLerpCount=::NewEnumerator0
      1:NiagaraFloat=8
      2:NiagaraFloat=&UniformRangedFloat(Minimum=12, Maximum=15)
      3:NiagaraFloat=8
      4:NiagaraFloat=?
      5:NiagaraFloat=?
      Alpha:NiagaraFloat=$Particles.AgeByParent
     )
     B=&Lerp_Float(A=0.9, B=1.125, Alpha=$Particles.DistanceBetweenParents)
    )
   )
   &InitializeParticle(
    Color=&ColorFromCurve(
     ColorCurve=DataInterface() {propertyValues {
       AlphaCurve {keys [<at 0 1>,<at 1 1>]}
       BlueCurve {keys [<at 0.0171 0>,<at 0.4894 0.0261>,<at 1.0012 1>]}
       GreenCurve {keys [<at 0.0171 0>,<at 0.4894 1>,<at 1.0012 0>]}
       RedCurve {keys [<at 0.0171 1>,<at 0.4894 0>,<at 1.0012 0>]}
      }}
     CurveIndex=$Particles.AgeByParent
    )
    Lifetime=0.7
    Module.Write Color=false
    Module.Write Lifetime=false
    Module.Write Mass=false
    Module.Write Material Random=true
    Module.Write Position=false
    Module.Write Scale=false
    Module.Write SpriteRotation=false
    Module.Write SpriteSize=true
    Sprite Size=&Vector2DFromFloat(Value=&Multiply_Float(A=1, B=1))
   ) {enabled=false}
   /SetVariables(
    Particles.OffsetPerRibbon:Vector3f=&UniformRangedVector(
     Minimum=&RandomVector(Vector Scale=-4)
     Maximum=&RandomVector(Vector Scale=4)
     Randomness Mode=::NewEnumerator1
     Random Seed=$Particles.ParentIDInt
     Fixed Random Seed=true
    )
    Particles.OffsetRandom:Vector3f=&RandomRangeVector(
     Minimum=&RandomVector(Vector Scale=-0.1)
     Maximum=&RandomVector(Vector Scale=0.1)
    )
    Particles.ParentsLerp:Vector3f=&Lerp_Vector(
     A=$Particles.ParentPosition
     B=$Particles.ParentFriendPosition
     Alpha=&ClampFloat(Float=&Add_Float(A=-0.5, B=&Multiply_Float(A=2, B=$Particles.AgeByParent)))
    )
   ) {isSetParameters=true}
   /SetVariables(Particles.Position=&Add_Vector(
     A=$Particles.ParentsLerp
     B=&Multiply_VectorByFloat(
      Vector=&Add_Vector(A=$Particles.OffsetPerRibbon, B=$Particles.OffsetRandom)
      Float=&FloatFromCurve(
       FloatCurve=DataInterface() {propertyValues {Curve {keys [
           {arriveTangent=2.9233, interpMode=RCIM_Cubic, leaveTangent=2.9233, tangentMode=RCTM_User, time=0.1, value=0}
           {interpMode=RCIM_Cubic, tangentMode=RCTM_User, time=0.4951, value=1}
           {arriveTangent=-2.7178, interpMode=RCIM_Cubic, leaveTangent=-2.7178, tangentMode=RCTM_User, time=0.9, value=0}
          ]}}}
       CurveIndex=$Particles.AgeByParent
      )
     )
    )) {isSetParameters=true}
   &AddVelocity(Velocity=&Multiply_VectorByFloat(
     Vector=&Multiply_Vector(A=$Particles.OffsetRandom, B=<-40 10 -40>)
     Float=&FloatFromCurve(
      FloatCurve=DataInterface() {propertyValues {Curve {keys [
          {arriveTangent=0.0416, interpMode=RCIM_Cubic, leaveTangent=0.0416, tangentMode=RCTM_User, time=0, value=0.2}
          {interpMode=RCIM_Cubic, tangentMode=RCTM_User, time=0.5, value=1}
          {arriveTangent=-0.0664, interpMode=RCIM_Cubic, leaveTangent=-0.0664, tangentMode=RCTM_User, time=1, value=0.2}
         ]}}}
      CurveIndex=$Particles.AgeByParent
      Scale Curve=$Particles.DistanceBetweenParents
     )
    ))
  ]
  particleUpdateScript [
   &ParticleState()
   &Drag(Drag=2, Use Linear Drag=true, Use Rotational Drag=false)
   &SolveForcesAndVelocity(Clamp Velocity=false, Limit Acceleration=false)
   &ScaleRibbonWidth(Ribbon Width Scale=&Lerp_Float(
     A=&FloatFromCurve(FloatCurve=DataInterface() {propertyValues {Curve {keys [
          {arriveTangent=3.4726, interpMode=RCIM_Cubic, leaveTangent=3.4726, tangentMode=RCTM_User, time=0, value=0}
          {arriveTangent=3.0838, interpMode=RCIM_Cubic, leaveTangent=3.0838, tangentMode=RCTM_User, time=0.7674, value=0.5898}
          <at 0.8252 0.9565>
          {arriveTangent=-0.254, interpMode=RCIM_Cubic, leaveTangent=-0.254, tangentMode=RCTM_User, time=1, value=0}
         ]}}})
     B=&FloatFromCurve(FloatCurve=DataInterface() {propertyValues {Curve {keys [
          {arriveTangent=1.1373, interpMode=RCIM_Cubic, leaveTangent=1.1373, tangentMode=RCTM_User, time=0, value=0}
          <at 0.0703 0.8282>
          {arriveTangent=-1.7167, interpMode=RCIM_Cubic, leaveTangent=-1.7167, tangentMode=RCTM_User, time=0.174, value=0.4523}
          {arriveTangent=-2.7266, interpMode=RCIM_Cubic, leaveTangent=-2.7266, tangentMode=RCTM_User, time=1, value=0}
         ]}}})
     Alpha=/Niagara/DynamicInputs/Integer/V2/Round_Float(Float:NiagaraFloat=$Particles.Initial.MaterialRandom)
    ))
   &ScaleColor(
    Scale RGB=&VectorFromFloat(Value=&LerpMultipleFloats(
      Float Count:ENiagara_MultipleLerpCount=::NewEnumerator2
      1:NiagaraFloat=1.5
      2:NiagaraFloat=0.5
      3:NiagaraFloat=0.5
      4:NiagaraFloat=0.5
      5:NiagaraFloat=0.1
      Alpha:NiagaraFloat=$Particles.NormalizedAge
     ))
    ScaleA=false
    ScaleRGB=true
   ) {module=ScaleColor001}
  ]
  rendererClasses [&NiagaraRibbonRendererProperties, &NiagaraSpriteRendererProperties]
  renderers [
   {rendererClass=&NiagaraRibbonRendererProperties}
   {rendererClass=&NiagaraSpriteRendererProperties, rendererIndex=1}
  ]
 }
}
summary {
 systemName=NS_WizardDark_Feather
 emitters [
  {
   emitterName=BigLightningRibbon
   rendererClasses [&NiagaraRibbonRendererProperties, &NiagaraSpriteRendererProperties]
  }
  {
   emitterName=BigLightningEndPicker
   rendererClasses [&NiagaraSpriteRendererProperties, &NiagaraLightRendererProperties]
  }
 ]
 userVariables [
  User.NumberKills {type=NiagaraFloat, defaultValue=0, description=""}
  User.SkeletalMesh {
   type=NiagaraDataInterfaceSkeletalMesh
   defaultValue {
    struct=NiagaraExt_VariableValue_DataInterface
    value {dataInterface=&self:NiagaraDataInterfaceSkeletalMesh_0}
   }
  }
 ]
}
